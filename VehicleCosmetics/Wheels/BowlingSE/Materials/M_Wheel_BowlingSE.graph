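# Material for the BowlingSE cosmetic vehicle wheel, adding animated glowing tire and rim effects to MF_WheelParent_Simple.
# The tire uses an RGB mask to mix a panning, optionally painted texture with a sine pulse in base color and emissive, plus optional Fresnel tint.
# The rim's emissive pulses on a sine phase-offset by its mask.
# Vertex color red blends rim and tire before the shared damage and de-rez transition functions.
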Material /VehicleCosmetics/Wheels/BowlingSE/Materials/M_Wheel_BowlingSE
graph {
 properties {
  bAutomaticallySetUsageInEditor=false
  bUseMaterialAttributes=true
  bUsedWithMeshDeformer=true
  bUsedWithSkeletalMesh=true
  pixelDepthOffsetMode=PDOM_Legacy
  usedShadingModels=Default Lit
 }
 nodes [
  Composite() {subgraphName=Color Settings A-C}
  Composite() {subgraphName=Normal w/Tread Mirroring}
  Composite() {subgraphName=Default Settings}
  n1 = MaterialFunctionCall() {materialFunction=&MF_WheelParent_Simple}
  n2 = GetMaterialAttributes(None=@n1.MaterialAttributes) {attributeGetTypes [#1]}
  n3 = Panner() {speedX=-0.5}
  n4 = TextureSampleParameter2D(UVs=@n3) {
   desc=Diffuse Input Texture - Optional
   group=1_Input Textures
   parameterName=Tire_DiffuseTexture
   texture=/VehicleCosmetics/Wheels/BowlingSE/Textures/T_Tire_BowlingSE_BC
  }
  n5 = TextureSampleParameter2D(UVs=@n3) {group=1_Input Textures, parameterName=Tire_MaskRGBA, samplerType=SAMPLERTYPE_Masks, texture=/VehicleCosmetics/Textures/T_Black_Masks}
  n6 = Multiply(A=@n5.R, B=VectorParameter() {parameterName=Tire_ColorA, defaultValue=<rgba 1 0.1248 0 1>}.RGB)
  n7 = Multiply(A=@n5.G, B=VectorParameter() {parameterName=Tire_ColorB, defaultValue=<rgba 0 1 0.5029 1>}.RGB)
  n8 = StaticSwitchParameter(True=LinearInterpolate(A=@n4.RGB, B=Add(A=@n6, B=@n7), Alpha=Add(A=@n5.R, B=@n5.R)), False=@n4.RGB) {group=0_Animated, parameterName=isPainted}
  n9 = TextureSampleParameter2D() {
   desc=Animated Texture - Optional
   group=0_Animated
   parameterName=Tire_MaskAnimatedTexture
   samplerType=SAMPLERTYPE_LinearColor
   texture=/VehicleCosmetics/Wheels/BowlingSE/Textures/T_Tire_BowlingSE_RGB
  }
  n10 = Time()
  n11 = Multiply(A=Clamp(None=Sine(None=@n10) {period=2}) {minDefault=0.1}, B=@n9.R)
  n12 = Multiply(
   A=LinearInterpolate(A=LinearInterpolate(A=@n2.Base Color, B=@n8, Alpha=@n9.G), B=Multiply(A=@n2.Base Color, B=@n11), Alpha=@n9.R)
   B=ScalarParameter() {defaultValue=2, group=0_Animated, parameterName=Tire_EmissiveBrightness}
  )
  n13 = BlendMaterialAttributes(
   A=@n1.MaterialAttributes
   B=SetMaterialAttributes!(MaterialAttributes=@n1.MaterialAttributes, Emissive Color=@n12, Base Color=@n12) {attributeSetTypes [#2, #1]}
   Alpha=Add(A=@n9.R, B=@n9.R)
  )
  Composite!("In (MA)"=@n13, TireAO=@n1.TireRGBAMask, TireVCMask=@n1.TireRGBAMask) {subgraphName=Tire Fresnel}
  n14 = TextureSampleParameter2D() {
   desc=Animated Texture - Optional
   group=0_Animated
   parameterName=Rim_MaskAnimatedTexture
   texture=/VehicleCosmetics/Wheels/BowlingSE/Textures/T_Rim_BowlingSE_Animated_Mask
  }
  n15 = StaticSwitchParameter(True=@n14.B, False=@n14.B) {group=0_Animated, parameterName=LOD}
  n16 = Add(A=Multiply(A=@n10) {constB=0.5}, B=@n15)
  n17 = OneMinus(None=ConstantBiasScale(None=Sine(None=@n16)) {bias=1.5, scale=1})
  n18 = ScalarParameter() {
   defaultValue=3
   desc="R: Animated Emisive\r\nG: Emissive\r\nB: Animated Emisive for LOD\r\nAlpha: Emissive for LOD"
   group=0_Animated
   parameterName=Rim_EmissiveBrightness
  }
  n19 = StaticSwitchParameter(True=@n14.A, False=@n14.A) {group=0_Animated, parameterName=LOD}
  n20 = LinearInterpolate(A=Multiply(A=Clamp(None=@n17), B=@n18), B=@n18, Alpha=@n19)
  n21 = Multiply(B=@n15) {constA=15, constB=10}
  n22 = Clamp(None=MaterialFunctionCall(UV Channel=Add(A=@n21, B=@n19)) {materialFunction=&LinearGradient}.UGradient)
  n23 = SetMaterialAttributes!(MaterialAttributes=@n1.MaterialAttributes, Emissive Color=Multiply(A=@n2.Base Color, B=Multiply(A=@n20, B=@n22))) {attributeSetTypes [#2]}
  MaterialExpressionReroute_14 = Reroute(None=@n13)
  n24 = GetMaterialAttributes(None=@MaterialExpressionReroute_14."In (MA)") {attributeGetTypes [#1]}
  n25 = VectorParameter() {
   desc=Color fill for R value of RGB Mask
   group=3_Tire Properties
   parameterName=Tire_ColorA
   channelNames {a=Unused, b="", g="", r=""}
   defaultValue=<rgba 1 0.1248 0 1>
  }
  MaterialExpressionReroute_6 = Reroute(None=@n25)
  n26 = TextureSampleParameter2D(UVs=NamedRerouteUsage()) {group=1_Input Textures, parameterName=Tire_Normal, samplerType=SAMPLERTYPE_Normal, texture=/VehicleCosmetics/Textures/T_BlankNormal_B}
  n27 = MaterialFunctionCall(R=Multiply(A=@n26.R, B=NamedRerouteUsage()) {constB=-1}, G=Multiply(A=@n26.G, B=NamedRerouteUsage()) {constB=-1}, B=@n26.B) {materialFunction=&AppendMany}
  MaterialExpressionReroute_12 = Reroute(None=@n27.RGB)
  n28 = CameraVectorWS()
  n29 = Power(
   Base=DotProduct(
    A=@MaterialExpressionReroute_12.Tire_Normal
    B=Transform(None=@n28) {transformSourceType=TRANSFORMSOURCE_World, transformType=TRANSFORM_Tangent}
   )
   Exp=ScalarParameter() {defaultValue=6, group=5_Special Effects, parameterName=Tire_FresnelExponent, sortPriority=50}
  )
  n30 = Saturate(None=Multiply(
    A=ConstantBiasScale(None=@n29) {bias=0.0156}
    B=ScalarParameter() {defaultValue=1, group=5_Special Effects, parameterName=Tire_FresnelPower, sortPriority=50}
   ))
  MaterialExpressionReroute_15 = Reroute(None=@n1.TireRGBAMask)
  n31 = ComponentMask(None=@MaterialExpressionReroute_15.TireAO) {a=true, desc=Tire AO}
  n32 = ScalarParameter() {
   defaultValue=0.7
   desc=General Roughness that takes effect in absence of masking.
   group=4_Surface Defaults
   parameterName=Tire_RoughnessBase
   sliderMax=1
  }
  n33 = ScalarParameter() {desc=General Metalness that takes effect in absence of masking., group=4_Surface Defaults, parameterName=Tire_MetallicBase, sliderMax=1}
  n34 = ScalarParameter() {
   defaultValue=1
   desc="0==False, 1==True: AO gradient affects Emissive values. Use this for Emissives with depth. Values between 0 && 1 are allowed."
   group=5_Special Effects
   parameterName=Tire_AOTintsEmissive
  }
  n35 = ScalarParameter() {
   desc="0==False, 1==True: Blends AO with Diffuse. Values between 0 && 1 are allowed."
   group=5_Special Effects
   parameterName=Tire_AOTintsDiffuse
  }
  n36 = MaterialFunctionCall(R=@n32, G=@n33, B=@n34, A=@n35) {desc="Vector4 container for Rough, Metallic, AO settings", materialFunction=&AppendMany}
  MaterialExpressionReroute_13 = Reroute(None=@n36.RGBA)
  n37 = Saturate(None=ComponentMask(None=@MaterialExpressionReroute_13.RoughMetalAOTint) {a=true})
  n38 = LinearInterpolate(
   A=@n24.Base Color
   B=Multiply(
    A=ScalarParameter() {defaultValue=16, group=5_Special Effects, parameterName=Tire_FresnelBrightness, sortPriority=50}
    B=@MaterialExpressionReroute_6.ColorA
   )
   Alpha=LinearInterpolate(A=@n30, B=Multiply(A=@n31, B=@n30), Alpha=NamedRerouteDeclaration(None=@n37) {name=Bool_AO_Diffuse, nodeColor=<rgba 1 0.5859 0 1>})
  )
  MaterialExpressionReroute_16 = Reroute(None=@n1.VertexColorR)
  n39 = StaticSwitchParameter(
   True=BlendMaterialAttributes(
    A=@MaterialExpressionReroute_14."In (MA)"
    B=SetMaterialAttributes!(MaterialAttributes=@n24.MaterialAttributes, Base Color=@n38) {attributeSetTypes [#1]}
    Alpha=@MaterialExpressionReroute_16.TireVCMask
   )
   False=@MaterialExpressionReroute_14."In (MA)"
  ) {desc="Tire == True, No Tire == False. False is default.", group=5_Special Effects, parameterName=Tire_Exists}
  MaterialExpressionReroute_17 = Reroute(None=@n39)
  n40 = VertexColor()
  MaterialFunctionCall(In=MaterialFunctionCall(Material=BlendMaterialAttributes(A=@n23, B=@MaterialExpressionReroute_17.Output, Alpha=@n40.R)) {materialFunction=&MF_BaseVehicle_Damage}.Result) {materialFunction=&MF_VehicleDeRez_Transition}
  n41 = MaterialFunctionCall() {
   desc="UV - For all Tread Mirrors. Lerp A && B - For Normals"
   materialFunction=/VehicleCosmetics/Wheels/_SharedMaterials/MF_Wheel_TreadMirroring
  }
  NamedRerouteDeclaration(None=@n41.LerpB) {name=Tread_LerpB, nodeColor=<rgba 0 1 0.6953 1>}
  NamedRerouteDeclaration(None=@n41.LerpA) {name=Tread_LerpA, nodeColor=<rgba 0 1 0.6953 1>}
  NamedRerouteDeclaration(None=@n41.UV) {name=Tread_UV, nodeColor=<rgba 0 1 0.6953 1>}
  n42 = VectorParameter() {
   desc="Emissive, Roughness, Metallic values for ColorA. Emissive has no limit, Roughness && Metallic are clamped between 0 && 1."
   group=3_Tire Properties
   parameterName=Tire_ColorA_PBR
   channelNames {a=Unused, b=Metallic, g=Roughness, r=Emissive}
   defaultValue=<rgba 0 0.7 0 1>
  }
  n43 = VectorParameter() {
   desc=Color fill for G value of RGB Mask
   group=3_Tire Properties
   parameterName=Tire_ColorB
   channelNames {a=Unused, b="", g="", r=""}
   defaultValue=<rgba 0 1 0.5029 1>
  }
  n44 = VectorParameter() {
   desc="Emissive, Roughness, Metallic values for ColorB. Emissive has no limit, Roughness && Metallic are clamped between 0 && 1."
   group=3_Tire Properties
   parameterName=Tire_ColorB_PBR
   channelNames {a=Unused, b=Metallic, g=Roughness, r=Emissive}
   defaultValue=<rgba 0 0.7 0 1>
  }
  n45 = VectorParameter() {
   desc=Color fill for B value of RGB Mask
   group=3_Tire Properties
   parameterName=Tire_ColorC
   channelNames {a=Unused, b="", g="", r=""}
   defaultValue=<rgba 0.5029 0 1 1>
  }
  n46 = VectorParameter() {
   desc="Emissive, Roughness, Metallic values for ColorC. Emissive has no limit, Roughness && Metallic are clamped between 0 && 1."
   group=3_Tire Properties
   parameterName=Tire_ColorC_PBR
   channelNames {a=Unused, b=Metallic, g=Roughness, r=Emissive}
   defaultValue=<rgba 0 0.7 0 1>
  }
  PinBase!(ColorA=@n25, ColorA_PBR=@n42, ColorB=@n43, ColorB_PBR=@n44, ColorC=@n45, ColorC_PBR=@n46) {reroutePins [
    ColorA {expression=&self:MaterialExpressionReroute_6}
    ColorA_PBR {expression=&self:MaterialExpressionReroute_7}
    ColorB {expression=&self:MaterialExpressionReroute_8}
    ColorB_PBR {expression=&self:MaterialExpressionReroute_9}
    ColorC {expression=&self:MaterialExpressionReroute_10}
    ColorC_PBR {expression=&self:MaterialExpressionReroute_11}
   ]}
  PinBase!()
  PinBase()
  PinBase!(Tire_Normal=@n27.RGB) {reroutePins [Tire_Normal {expression=&self:MaterialExpressionReroute_12}]}
  PinBase!()
  PinBase!(RoughMetalAOTint=@n36.RGBA) {reroutePins [RoughMetalAOTint {expression=&self:MaterialExpressionReroute_13}]}
  PinBase!() {reroutePins [
    "In (MA)" {expression=&self:MaterialExpressionReroute_14}
    TireAO {expression=&self:MaterialExpressionReroute_15}
    TireVCMask {expression=&self:MaterialExpressionReroute_16}
   ]}
  PinBase!(Output=@n39) {reroutePins [Output {expression=&self:MaterialExpressionReroute_17}]}
  MaterialExpressionReroute_10 = Reroute(None=@n45)
  MaterialExpressionReroute_11 = Reroute(None=@n46)
  MaterialExpressionReroute_7 = Reroute(None=@n42)
  MaterialExpressionReroute_8 = Reroute(None=@n43)
  MaterialExpressionReroute_9 = Reroute(None=@n44)
  StaticSwitchParameter(True=@MaterialExpressionReroute_17.Output, False=@n13) {group=0_Animated, parameterName=isPainted}
 ]
}
